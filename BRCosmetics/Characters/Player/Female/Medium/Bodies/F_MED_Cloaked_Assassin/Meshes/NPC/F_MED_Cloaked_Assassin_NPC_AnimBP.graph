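# Post-process style AnimBP for the Cloaked Assassin NPC body, copying the parent mesh pose and animating six floating crystals.
# Thigh-rotation alphas drive corrective bone offsets, then weighted constraints position arm crystals before a RigidBody pass.
# Time-based spin and sine/cosine bobbing move the crystals, cached as CubeCache.
# Two neck-relative AnimDynamics spring chains blend on bIsOnGround, with the airborne chain's alpha and external force set by bIsInFrontEnd.
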AnimBlueprint /BRCosmetics/Characters/Player/Female/Medium/Bodies/F_MED_Cloaked_Assassin/Meshes/NPC/F_MED_Cloaked_Assassin_NPC_AnimBP
target_skeleton=/BRCosmetics/Characters/Player/Female/Medium/Bodies/F_MED_Cloaked_Assassin/Meshes/F_MED_Cloaked_Assassin_Skeleton
graphs [
 AnimGraph {nodes [
   n1 = Anim Dynamics - Bone: dyn_crystal_01(ComponentPose=Local To Component(LocalPose=Use cached pose 'CubeCache'()), Alpha=0.5, ExternalForce="0,0,0") {
    angularDampingOverride=0.9
    angularSpringConstant=150
    bAngularSpring=true
    bGravityOverrideInSimSpace=true
    bLinearSpring=true
    bOverrideAngularDamping=true
    bOverrideLinearDamping=true
    gravityScale=0
    linearDampingOverride=0.995
    linearSpringConstant=50
    simulationSpace=BoneRelative
    alphaBoolBlend {blendOption=Linear}
    boundBone {boneName=dyn_crystal_01}
    componentLinearAccScale=<0.05 0.05 0.05>
    componentLinearVelScale=<0.8 0.8 0.6>
    physicsBodyDefinitions [{
      sphereCollisionRadius=0
      boundBone {boneName=dyn_crystal_01}
      boxExtents=<50 50 50>
      constraintSetup {angularLimitsMax=<0 0 0>, angularLimitsMin=<0 0 0>, angularTarget=<0 0 0>, linearAxesMax=<2 3 5>, linearAxesMin=<-2 -1 -5>}
      localJointOffset=<0 -3 0>
     }]
    relativeSpaceBone {boneName=neck_01}
   }
   n2 = Anim Dynamics - Bone: dyn_crystal_03(ComponentPose=@n1, Alpha=0.5, ExternalForce="0,0,0") {
    angularDampingOverride=0.9
    angularSpringConstant=150
    bAngularSpring=true
    bGravityOverrideInSimSpace=true
    bLinearSpring=true
    bOverrideAngularDamping=true
    bOverrideLinearDamping=true
    gravityScale=0
    linearDampingOverride=0.995
    linearSpringConstant=50
    simulationSpace=BoneRelative
    alphaBoolBlend {blendOption=Linear}
    boundBone {boneName=dyn_crystal_03}
    componentLinearAccScale=<0.1 0.1 0.3>
    componentLinearVelScale=<0.8 0.8 0.6>
    physicsBodyDefinitions [{
      sphereCollisionRadius=0
      boundBone {boneName=dyn_crystal_03}
      boxExtents=<50 50 50>
      constraintSetup {angularLimitsMax=<0 0 0>, angularLimitsMin=<0 0 0>, angularTarget=<0 0 0>, linearAxesMax=<2 5 8>, linearAxesMin=<-2 0 -8>}
      localJointOffset=<0 -3 0>
     }]
    relativeSpaceBone {boneName=neck_01}
   }
   n3 = Anim Dynamics - Bone: dyn_crystal_05(ComponentPose=@n2, Alpha=0.5, ExternalForce="0,0,0") {
    angularDampingOverride=0.9
    angularSpringConstant=150
    bAngularSpring=true
    bGravityOverrideInSimSpace=true
    bLinearSpring=true
    bOverrideAngularDamping=true
    bOverrideLinearDamping=true
    gravityScale=0
    linearDampingOverride=0.995
    linearSpringConstant=50
    simulationSpace=BoneRelative
    alphaBoolBlend {blendOption=Linear}
    boundBone {boneName=dyn_crystal_05}
    componentLinearAccScale=<0.05 0.05 0.05>
    componentLinearVelScale=<0.8 1 0.6>
    physicsBodyDefinitions [{
      sphereCollisionRadius=0
      boundBone {boneName=dyn_crystal_05}
      boxExtents=<50 50 50>
      constraintSetup {angularLimitsMax=<0 0 0>, angularLimitsMin=<0 0 0>, angularTarget=<0 0 0>, linearAxesMax=<2 5 8>, linearAxesMin=<-2 1 -8>}
      localJointOffset=<0 -3 0>
     }]
    relativeSpaceBone {boneName=neck_01}
   }
   n4 = Anim Dynamics - Bone: dyn_crystal_02(ComponentPose=@n3, Alpha=0.5, ExternalForce="0,0,0") {
    angularDampingOverride=0.9
    angularSpringConstant=150
    bAngularSpring=true
    bGravityOverrideInSimSpace=true
    bLinearSpring=true
    bOverrideAngularDamping=true
    bOverrideLinearDamping=true
    gravityScale=0
    linearDampingOverride=0.995
    linearSpringConstant=50
    simulationSpace=BoneRelative
    alphaBoolBlend {blendOption=Linear}
    boundBone {boneName=dyn_crystal_02}
    componentLinearAccScale=<0.1 0.1 0.3>
    componentLinearVelScale=<0.8 0.8 0.6>
    physicsBodyDefinitions [{
      sphereCollisionRadius=0
      boundBone {boneName=dyn_crystal_02}
      boxExtents=<50 50 50>
      constraintSetup {angularLimitsMax=<0 0 0>, angularLimitsMin=<0 0 0>, angularTarget=<0 0 0>, linearAxesMax=<2 5 8>, linearAxesMin=<-2 -1 -8>}
      localJointOffset=<0 -3 0>
     }]
    relativeSpaceBone {boneName=neck_01}
   }
   n5 = Anim Dynamics - Bone: dyn_crystal_04(ComponentPose=@n4, Alpha=0.5, ExternalForce="0,0,0") {
    angularDampingOverride=0.9
    angularSpringConstant=150
    bAngularSpring=true
    bGravityOverrideInSimSpace=true
    bLinearSpring=true
    bOverrideAngularDamping=true
    bOverrideLinearDamping=true
    gravityScale=0
    linearDampingOverride=0.995
    linearSpringConstant=50
    simulationSpace=BoneRelative
    alphaBoolBlend {blendOption=Linear}
    boundBone {boneName=dyn_crystal_04}
    componentLinearAccScale=<0.05 0.05 0.05>
    componentLinearVelScale=<0.8 0.8 0.6>
    physicsBodyDefinitions [{
      sphereCollisionRadius=0
      boundBone {boneName=dyn_crystal_04}
      boxExtents=<50 50 50>
      constraintSetup {angularLimitsMax=<0 0 0>, angularLimitsMin=<0 0 0>, angularTarget=<0 0 0>, linearAxesMax=<2 5 8>, linearAxesMin=<-2 0 -8>}
      localJointOffset=<0 -3 0>
     }]
    relativeSpaceBone {boneName=neck_01}
   }
   n6 = Anim Dynamics - Bone: dyn_crystal_06(ComponentPose=@n5, Alpha=0.5, ExternalForce="0,0,0") {
    angularDampingOverride=0.9
    angularSpringConstant=150
    bAngularSpring=true
    bGravityOverrideInSimSpace=true
    bLinearSpring=true
    bOverrideAngularDamping=true
    bOverrideLinearDamping=true
    gravityScale=0
    linearDampingOverride=0.995
    linearSpringConstant=50
    simulationSpace=BoneRelative
    alphaBoolBlend {blendOption=Linear}
    boundBone {boneName=dyn_crystal_06}
    componentLinearAccScale=<0.1 0.1 0.3>
    componentLinearVelScale=<0.8 1 0.6>
    physicsBodyDefinitions [{
      sphereCollisionRadius=0
      boundBone {boneName=dyn_crystal_06}
      boxExtents=<50 50 50>
      constraintSetup {angularLimitsMax=<0 0 0>, angularLimitsMin=<0 0 0>, angularTarget=<0 0 0>, linearAxesMax=<2 5 8>, linearAxesMin=<-2 1 -8>}
      localJointOffset=<0 -3 0>
     }]
    relativeSpaceBone {boneName=neck_01}
   }
   n7 = Get Frontend() -> (Frontend:double)
   n8 = Get VelocityAlpha() -> (VelocityAlpha:Vector)
   n9 = Anim Dynamics - Bone: dyn_crystal_01(ComponentPose=Local To Component(LocalPose=Use cached pose 'CubeCache'()), Alpha=@n7, ExternalForce=@n8) {
    angularDampingOverride=0.9
    angularSpringConstant=210
    bAngularSpring=true
    bGravityOverrideInSimSpace=true
    bLinearSpring=true
    bOverrideAngularDamping=true
    bOverrideLinearDamping=true
    gravityScale=0
    linearDampingOverride=0.9999
    linearSpringConstant=120
    simulationSpace=BoneRelative
    alphaBoolBlend {blendOption=Linear}
    boundBone {boneName=dyn_crystal_01}
    componentAppliedLinearAccClamp=<6000 6000 6000>
    componentLinearAccScale=<0.05 0.05 0.05>
    componentLinearVelScale=<0.9 0.9 0.9>
    gravityOverride {z=-980}
    physicsBodyDefinitions [{
      sphereCollisionRadius=0
      boundBone {boneName=dyn_crystal_01}
      boxExtents=<50 50 50>
      constraintSetup {angularLimitsMax=<0 0 0>, angularLimitsMin=<0 0 0>, angularTarget=<0 0 0>, linearAxesMax=<3 3 3>, linearAxesMin=<-3 -1 -3>}
      localJointOffset=<-3 0 0>
     }]
    relativeSpaceBone {boneName=neck_01}
   }
   n10 = Anim Dynamics - Bone: dyn_crystal_03(ComponentPose=@n9, Alpha=@n7, ExternalForce=@n8) {
    angularDampingOverride=0.9
    angularSpringConstant=210
    bAngularSpring=true
    bGravityOverrideInSimSpace=true
    bLinearSpring=true
    bOverrideAngularDamping=true
    bOverrideLinearDamping=true
    gravityScale=0
    linearDampingOverride=0.9999
    linearSpringConstant=210
    simulationSpace=BoneRelative
    alphaBoolBlend {blendOption=Linear}
    boundBone {boneName=dyn_crystal_03}
    componentAppliedLinearAccClamp=<6000 6000 6000>
    componentLinearVelScale=<0.9 0.9 0.9>
    gravityOverride {z=-980}
    physicsBodyDefinitions [{
      sphereCollisionRadius=0
      boundBone {boneName=dyn_crystal_03}
      boxExtents=<50 50 50>
      constraintSetup {angularLimitsMax=<0 0 0>, angularLimitsMin=<0 0 0>, angularTarget=<0 0 0>, linearAxesMax=<2 3 4>, linearAxesMin=<-2 0 -4>}
      localJointOffset=<-3 0 0>
     }]
    relativeSpaceBone {boneName=neck_01}
   }
   n11 = Anim Dynamics - Bone: dyn_crystal_05(ComponentPose=@n10, Alpha=@n7, ExternalForce=@n8) {
    angularDampingOverride=0.9
    angularSpringConstant=210
    bAngularSpring=true
    bGravityOverrideInSimSpace=true
    bLinearSpring=true
    bOverrideAngularDamping=true
    bOverrideLinearDamping=true
    gravityScale=0
    linearDampingOverride=0.9999
    linearSpringConstant=210
    simulationSpace=BoneRelative
    alphaBoolBlend {blendOption=Linear}
    boundBone {boneName=dyn_crystal_05}
    componentAppliedLinearAccClamp=<6000 6000 6000>
    componentLinearAccScale=<0.05 0.05 0.05>
    componentLinearVelScale=<0.9 0.9 0.9>
    gravityOverride {z=-980}
    physicsBodyDefinitions [{
      sphereCollisionRadius=0
      boundBone {boneName=dyn_crystal_05}
      boxExtents=<50 50 50>
      constraintSetup {angularLimitsMax=<0 0 0>, angularLimitsMin=<0 0 0>, angularTarget=<0 0 0>, linearAxesMax=<2 3 4>, linearAxesMin=<-2 1 -4>}
      localJointOffset=<-3 0 0>
     }]
    relativeSpaceBone {boneName=neck_01}
   }
   n12 = Anim Dynamics - Bone: dyn_crystal_02(ComponentPose=@n11, Alpha=@n7, ExternalForce=@n8) {
    angularDampingOverride=0.9
    angularSpringConstant=210
    bAngularSpring=true
    bGravityOverrideInSimSpace=true
    bLinearSpring=true
    bOverrideAngularDamping=true
    bOverrideLinearDamping=true
    gravityScale=0
    linearDampingOverride=0.9999
    linearSpringConstant=210
    simulationSpace=BoneRelative
    alphaBoolBlend {blendOption=Linear}
    boundBone {boneName=dyn_crystal_02}
    componentAppliedLinearAccClamp=<6000 6000 6000>
    componentLinearVelScale=<0.9 0.9 0.9>
    gravityOverride {z=-980}
    physicsBodyDefinitions [{
      sphereCollisionRadius=0
      boundBone {boneName=dyn_crystal_02}
      boxExtents=<50 50 50>
      constraintSetup {angularLimitsMax=<0 0 0>, angularLimitsMin=<0 0 0>, angularTarget=<0 0 0>, linearAxesMax=<2 3 4>, linearAxesMin=<-2 -1 -4>}
      localJointOffset=<-3 0 0>
     }]
    relativeSpaceBone {boneName=neck_01}
   }
   n13 = Anim Dynamics - Bone: dyn_crystal_04(ComponentPose=@n12, Alpha=@n7, ExternalForce=@n8) {
    angularDampingOverride=0.9
    angularSpringConstant=210
    bAngularSpring=true
    bGravityOverrideInSimSpace=true
    bLinearSpring=true
    bOverrideAngularDamping=true
    bOverrideLinearDamping=true
    gravityScale=0
    linearDampingOverride=0.9999
    linearSpringConstant=165
    simulationSpace=BoneRelative
    alphaBoolBlend {blendOption=Linear}
    boundBone {boneName=dyn_crystal_04}
    componentAppliedLinearAccClamp=<6000 6000 6000>
    componentLinearAccScale=<0.05 0.05 0.05>
    componentLinearVelScale=<0.9 0.9 0.9>
    gravityOverride {z=-980}
    physicsBodyDefinitions [{
      sphereCollisionRadius=0
      boundBone {boneName=dyn_crystal_04}
      boxExtents=<50 50 50>
      constraintSetup {angularLimitsMax=<0 0 0>, angularLimitsMin=<0 0 0>, angularTarget=<0 0 0>, linearAxesMax=<2 3 2>, linearAxesMin=<-2 0 -2>}
      localJointOffset=<-3 0 0>
     }]
    relativeSpaceBone {boneName=neck_01}
   }
   n14 = Anim Dynamics - Bone: dyn_crystal_06(ComponentPose=@n13, Alpha=@n7, ExternalForce=@n8) {
    angularDampingOverride=0.9
    angularSpringConstant=210
    bAngularSpring=true
    bGravityOverrideInSimSpace=true
    bLinearSpring=true
    bOverrideAngularDamping=true
    bOverrideLinearDamping=true
    gravityScale=0
    linearDampingOverride=0.9999
    linearSpringConstant=195
    simulationSpace=BoneRelative
    alphaBoolBlend {blendOption=Linear}
    boundBone {boneName=dyn_crystal_06}
    componentAppliedLinearAccClamp=<6000 6000 6000>
    componentLinearAccScale=<0.1 0.1 0.1>
    componentLinearVelScale=<0.9 0.9 0.9>
    gravityOverride {z=-980}
    physicsBodyDefinitions [{
      sphereCollisionRadius=0
      boundBone {boneName=dyn_crystal_06}
      boxExtents=<50 50 50>
      constraintSetup {angularLimitsMax=<0 0 0>, angularLimitsMin=<0 0 0>, angularTarget=<0 0 0>, linearAxesMax=<4 3 2>, linearAxesMin=<-4 0 -2>}
      localJointOffset=<-3 0 0>
     }]
    relativeSpaceBone {boneName=neck_01}
   }
   Output Pose(Result=Blend Poses by bool(
     BlendPose_0=Component To Local(ComponentPose=@n6)
     BlendPose_1=Component To Local(ComponentPose=@n14)
     BlendTime_0=0.5
     BlendTime_1=0.5
     bActiveValue=Get bIsOnGround() -> (bIsOnGround:bool)
    ) {blendType=Linear}) {layerGroup=None}
   n15 = "Transform (Modify) Bone - Bone: dyn_sub_thigh_r"(
    ComponentPose=Local To Component(LocalPose=Copy Pose From Mesh() {bUseAttachedParent=true})
    Alpha=Get Thigh_Rot_R() -> (Thigh_Rot_R:double)
    Translation="20,-10,-10"
    Rotation="0,0,0"
    Scale="0,0,0"
   ) {
    translationMode=BMM_Additive
    translationSpace=BCS_BoneSpace
    alphaBoolBlend {blendOption=Linear}
    boneToModify {boneName=dyn_sub_thigh_r}
    scale=<0 0 0>
    translation=<20 -10 -10>
   }
   n16 = "Transform (Modify) Bone - Bone: dyn_sub_thigh_l"(ComponentPose=@n15, Alpha=Get Thigh_Rot_L() -> (Thigh_Rot_L:double), Translation="20,-10,-10", Rotation="0,0,0", Scale="0,0,0") {
    translationMode=BMM_Additive
    translationSpace=BCS_BoneSpace
    alphaBoolBlend {blendOption=Linear}
    boneToModify {boneName=dyn_sub_thigh_l}
    scale=<0 0 0>
    translation=<20 -10 -10>
   }
   n17 = "Constraint - dyn_crystal_arm_l (PPPPP)"!(
    ComponentPose=@n16
    Alpha=1
    ConstraintWeights_0=0.1395
    ConstraintWeights_1:float=0.0147
    ConstraintWeights_2:float=0.1271
    ConstraintWeights_3:float=0.7183
    ConstraintWeights_4:float=0.0004
   ) -> (Pose) {
    alphaBoolBlend {blendOption=Linear}
    boneToModify {boneName=dyn_crystal_arm_l}
    constraintSetup [
     {targetBone {boneName=upperarm_twist_02_l}}
     {targetBone {boneName=clavicle_l}}
     {targetBone {boneName=upperarm_l}}
     {targetBone {boneName=upperarm_twist_01_l}}
     {targetBone {boneName=spine_05}}
    ]
    constraintWeights [1,1,1,1,1]
   }
   n18 = "Constraint - dyn_crystal_arm_l (PPPPP)"(
    ComponentPose=@n17
    Alpha=1
    ConstraintWeights_0=0.203
    ConstraintWeights_1=0.0001
    ConstraintWeights_2=0.1946
    ConstraintWeights_3=0.6022
    ConstraintWeights_4=0
   ) {
    alphaBoolBlend {blendOption=Linear}
    boneToModify {boneName=dyn_crystal_arm_l}
    constraintSetup [
     {targetBone {boneName=upperarm_twist_02_l}}
     {targetBone {boneName=clavicle_l}}
     {targetBone {boneName=upperarm_l}}
     {targetBone {boneName=upperarm_twist_01_l}}
     {targetBone {boneName=spine_05}}
    ]
    constraintWeights [1,1,1,1,1]
   }
   n19 = "Constraint - dyn_crystal_arm_l (PPP)"!(ComponentPose=@n18, Alpha=1, ConstraintWeights_0=0.2014, ConstraintWeights_1:float=0.2033, ConstraintWeights_2:float=0.5954) -> (Pose) {
    alphaBoolBlend {blendOption=Linear}
    boneToModify {boneName=dyn_crystal_arm_l}
    constraintSetup [{targetBone {boneName=upperarm_twist_02_l}}, {targetBone {boneName=upperarm_l}}, {targetBone {boneName=upperarm_twist_01_l}}]
    constraintWeights [1,1,1]
   }
   n20 = "Constraint - dyn_crystal_arm_r (PPPPP)"!(
    ComponentPose=@n19
    Alpha=1
    ConstraintWeights_0=0.4
    ConstraintWeights_1:float=0
    ConstraintWeights_2:float=0
    ConstraintWeights_3:float=0.1
    ConstraintWeights_4:float=0.5
   ) -> (Pose) {
    alphaBoolBlend {blendOption=Linear}
    boneToModify {boneName=dyn_crystal_arm_r}
    constraintSetup [
     {targetBone {boneName=lowerarm_twist_02_r}}
     {targetBone {boneName=thumb_01_r}}
     {targetBone {boneName=hand_r}}
     {targetBone {boneName=lowerarm_r}}
     {targetBone {boneName=lowerarm_twist_01_r}}
    ]
    constraintWeights [1,1,1,1,1]
   }
   Save cached pose 'body_pose'(Pose=Component To Local(ComponentPose=@n20)) {cacheName=body_pose}
   n21 = RigidBody(ComponentPose=Local To Component(LocalPose=Use cached pose 'body_pose'())) {
    bClampLinearTranslationLimitToRefPose=true
    bOverrideWorldGravity=true
    alphaBoolBlend {blendOption=Linear}
    baseBoneRef {boneName=spine_05}
    componentAppliedLinearAccClamp=<3000 3000 3000>
    componentLinearAccScale=<0.2 0.2 0.2>
    componentLinearVelScale=<0.2 0.2 0.2>
    overrideWorldGravity {z=-980}
   }
   n22 = Get Spinning_Rotator() -> (Spinning_Rotator:Rotator)
   n23 = Get Cube_Float_A() -> (Cube_Float_A:Vector)
   n24 = "Transform (Modify) Bone - Bone: dyn_crystal_01"(ComponentPose=@n21, Alpha=1, Translation=@n23, Rotation=@n22, Scale="1,1,1") {rotationMode=BMM_Additive, rotationSpace=BCS_BoneSpace, alphaBoolBlend {blendOption=Linear}, boneToModify {boneName=dyn_crystal_01}}
   n25 = "Transform (Modify) Bone - Bone: dyn_crystal_02"(ComponentPose=@n24, Alpha=1, Translation=@n23, Rotation=@n22, Scale="1,1,1") {rotationMode=BMM_Additive, rotationSpace=BCS_BoneSpace, alphaBoolBlend {blendOption=Linear}, boneToModify {boneName=dyn_crystal_02}}
   n26 = Get Cube_Float_B() -> (Cube_Float_B:Vector)
   n27 = "Transform (Modify) Bone - Bone: dyn_crystal_03"(ComponentPose=@n25, Alpha=1, Translation=@n26, Rotation=@n22, Scale="1,1,1") {rotationMode=BMM_Additive, rotationSpace=BCS_BoneSpace, alphaBoolBlend {blendOption=Linear}, boneToModify {boneName=dyn_crystal_03}}
   n28 = "Transform (Modify) Bone - Bone: dyn_crystal_04"(ComponentPose=@n27, Alpha=1, Translation=@n26, Rotation=@n22, Scale="1,1,1") {rotationMode=BMM_Additive, rotationSpace=BCS_BoneSpace, alphaBoolBlend {blendOption=Linear}, boneToModify {boneName=dyn_crystal_04}}
   n29 = "Transform (Modify) Bone - Bone: dyn_crystal_05"(ComponentPose=@n28, Alpha=1, Translation=Get Cube_Float_C() -> (Cube_Float_C:Vector), Rotation=@n22, Scale="1,1,1") {rotationMode=BMM_Additive, rotationSpace=BCS_BoneSpace, alphaBoolBlend {blendOption=Linear}, boneToModify {boneName=dyn_crystal_05}}
   n30 = "Transform (Modify) Bone - Bone: dyn_crystal_06"(ComponentPose=@n29, Alpha=1, Translation=Get Cube_Float_D() -> (Cube_Float_D:Vector), Rotation=@n22, Scale="1,1,1") {rotationMode=BMM_Additive, rotationSpace=BCS_BoneSpace, alphaBoolBlend {blendOption=Linear}, boneToModify {boneName=dyn_crystal_06}}
   Save cached pose 'CubeCache'(Pose=Component To Local(ComponentPose=@n30)) {cacheName=CubeCache}
   Get Frontend()
  ]}
 EventGraph {nodes [
   Event BlueprintUpdateAnimation()
   > Get!(execute:Exec) -> (then:Exec, else:Exec, Skel_Mesh:Skeletal Mesh Component Object Reference)
   > n1 = Sequence!(execute) -> (then_0, then_1, then_2:Exec)
   > Set Thigh_Rot_L(Thigh_Rot_L:double=GetAlphaFromReferenceTransformDelta(SkeletalMeshComponent=Get Skel_Mesh(), BoneName=thigh_l, BaseName=pelvis, InputRange=(X=-19,Y=-101), TwistAxis="0,0,1")) -> (Output_Get:double)
   Set Thigh_Rot_R(
    execute=@n1.then_1
    Thigh_Rot_R:double=GetAlphaFromReferenceTransformDelta(SkeletalMeshComponent=Get Skel_Mesh(), BoneName=thigh_r, BaseName=pelvis, InputRange=(X=4,Y=-101), TwistAxis="0,0,1")
   ) -> (Output_Get:double)
   n2 = Sequence!(execute=@n1.then_2) -> (then_0, then_1, then_2, then_3:Exec)
   n3 = float * float(A=float / float(A=GetRealTimeSeconds(), B=-720) {is=title K2Node_CallFunction}, B=30000)
   > Set Spinning_Rotator(Spinning_Rotator:Rotator=MakeRotator(Roll=@n3, Yaw=@n3)) -> (Output_Get:Rotator)
   n4 = GetRealTimeSeconds()
   n5 = "Sin (Radians)"(A=float * float(A=float * float(A=1, B=1), B=@n4))
   n6 = "Sin (Radians)"(A=float * float(A=float * float(A=1, B=2), B=@n4))
   n7 = "Sin (Radians)"(A=float * float(A=float * float(A=1, B=1.5), B=@n4))
   Set Cube_Float_A(
    execute=@n2.then_1
    Cube_Float_A:Vector=MakeVector(X=float * float(A=@n7, B=0.3), Y=float * float(A=@n5, B=0.56), Z=float * float(A=@n6, B=0.22))
   ) -> (Output_Get:Vector)
   n8 = GetGameTimeInSeconds()
   n9 = "Cos (Radians)"!(A:double=float * float(A=float * float(A=1, B=1.5), B=@n8)) -> (ReturnValue:double)
   n10 = "Cos (Radians)"(A=float * float(A=float * float(A=1, B=1.78), B=@n8))
   n11 = "Cos (Radians)"(A=float * float(A=float * float(A=1, B=1.15), B=@n8))
   > Set Cube_Float_B(Cube_Float_B:Vector=MakeVector(X=float * float(A=@n9, B=0.6), Y=float * float(A=@n10, B=0.725), Z=float * float(A=@n11, B=0.3))) -> (Output_Get:Vector)
   n12 = GetGameTimeInSeconds()
   n13 = float * float(A=float * float(A=1, B=1.5), B=@n12)
   n14 = float * float(A=float * float(A=1, B=1), B=@n12)
   n15 = float * float(A=float * float(A=1, B=2), B=@n12)
   n16 = MakeVector(
    X=float * float(A="Sin (Radians)"(A=@n13), B=0.3)
    Y=float * float(A="Sin (Radians)"(A=@n14), B=0.6)
    Z=float * float(A="Sin (Radians)"(A=@n15), B=0.25)
   )
   > Set Cube_Float_C(Cube_Float_C:Vector=@n16) -> (Output_Get:Vector)
   n17 = GetGameTimeInSeconds()
   n18 = float * float(A=float * float(A=0.6678, B=1.5), B=@n17)
   n19 = float * float(A=float * float(A=1, B=1.47), B=@n17)
   n20 = float * float(A=float * float(A=1, B=0.953), B=@n17)
   n21 = MakeVector(
    X=float * float(A="Sin (Radians)"(A=@n18), B=0.85)
    Y=float * float(A="Sin (Radians)"(A=@n20), B=0.1)
    Z=float * float(A="Sin (Radians)"(A=@n19), B=0.7)
   )
   > Set Cube_Float_D(Cube_Float_D:Vector=@n21) -> (Output_Get:Vector)
   n22 = BreakVector(InVec=vector * float(A=@n21, B=5))
   > Set Floating_Rotator2(Floating_Rotator2:Rotator=MakeRotator(Roll=@n22, Pitch=@n22.Y, Yaw=@n22.Z)) -> (Output_Get:Rotator)
   n23 = BreakVector(InVec=vector * float(A=@n16, B=7))
   > Set Floating_Rotator1(Floating_Rotator1:Rotator=MakeRotator(Roll=@n23, Pitch=@n23.Y, Yaw=@n23.Z)) -> (Output_Get:Rotator)
   Set VelocityAlpha(execute=@n2.then_3, VelocityAlpha:Vector=SelectVector(A="500,500,500", B="0, 0, 0", bPickA=Get bIsInFrontEnd())) -> (Output_Get:Vector)
   Set Frontend(execute=@n2.then_2, Frontend:double=SelectFloat(A=1, B=0.5, bPickA=Get bIsInFrontEnd())) -> (Output_Get:double)
   Event BlueprintInitializeAnimation()
   > Set Skel_Mesh(Skel_Mesh=GetOwnerSkeletalMeshComponent())
  ]}
]
variables [
 Cube_Float_A {type=Vector}
 Cube_Float_B {type=Vector}
 Cube_Float_C {type=Vector}
 Cube_Float_D {type=Vector}
 Floating_Rotator1 {type=Rotator}
 Floating_Rotator2 {type=Rotator}
 Frontend {type=number}
 Skel_Mesh {type=SkeletalMeshComponent}
 Spinning_Rotator {type=Rotator}
 Thigh_Rot_L {type=number}
 Thigh_Rot_R {type=number}
 VelocityAlpha {type=Vector}
]
